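; A Blueprint interface for generic tech-art editor utility widgets, kept with the Blutility widget data.
; It declares a single function, onTabSelected, with no parameters shown and no implementation.
; Implementing widgets supply their own response to that call, which by its name signals a tab selection.
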(fn onTabSelected)
